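; Hookshot rope actor variant for HeatSnake that extends its parent Blueprint with reporting of hooked pickups.
; Its ReportAnalyticsPickupsHooked override calls the parent, then on authority finds the ChainWhip component on the Athena game state.
; If both are valid, it passes the instigator's player pawn and the pickup count to that component's NotifyPickupsHooked.
; Construction only calls the parent, and the event graph holds just unimplemented stubs.
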(event EventBeginPlay
  (|CallParentFunction) ; This node is disabled and will not be called.\nDrag off pins to build functionality.
)

(event Collision|EventActorBeginOverlap
  (params (OtherActor :type Actor%20Object%20Reference)))

(event EventTick
  (params (DeltaSeconds :type Float%20%28single-precision%29)))

(fn ReportAnalyticsPickupsHooked
  (params (NumOfPickups :type Integer))
  (bind _returnvalue (Fortnite|Athena|GameState|GetGameStateAthena))
  (bind _returnvalue_2 (Actor|GetComponentbyClass _returnvalue "/SunRoseChainWhipGameplay/Items/ChainWhip/B_ChainWhipGameStateComponent.B_ChainWhipGameStateComponent_C"))
  ;[0.00154 0.1615 0.155] Report analytics.
  (|CallParentFunction NumOfPickups)
  (Networking|SwitchHasAuthority
    (:Authority
      (Utilities|IsValid _returnvalue
        (:Is%20Valid
          (Utilities|IsValid _returnvalue_2
            (:Is%20Valid
              (Class|BChainWhipGameStateComponent|NotifyPickupsHooked _returnvalue_2 (Utilities|Casting|CastToFortPlayerPawn (Game|GetInstigator)) NumOfPickups))))))))

(fn ConstructionScript
  (|CallParentFunction))
